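# Material function for the QuartzBlob backpack that selects one scalar value per shading path and quality tier.
# A ShadingPathSwitch routes between two QualitySwitches: Default picks from Epic, High and Low inputs, Mobile from Med and Low.
# Input names map tiers to platforms, with notes on GBuffer, depth and outline limits.
# An unconnected dashed input separates desktop/console inputs from mobile ones, and unset tiers fall back to Epic or Mobile Med.
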MaterialFunction /Game/Accessories/FORT_Backpacks/Backpack_QuartzBlob/Material/MF_QuartzBlob_ScalarQualitySwitch
graph {nodes [
  FunctionInput() {bUsePreviewValueAsDefault=true, inputName=--------, inputType=FunctionInput_Scalar, sortPriority=3}
  n1 = FunctionInput() {
   bUsePreviewValueAsDefault=true
   desc="PC-High && Epic, XSX, PS5"
   description="PC-High && Epic, XSX, PS5"
   inputName="Epic - PC-High && Epic, XSX, PS5"
   inputType=FunctionInput_Scalar
  }
  n2 = FunctionInput() {
   bUsePreviewValueAsDefault=true
   desc="PC Low && Medium (not Perf mode)"
   description="PC Low && Medium (not PerfMode)"
   inputName="Low - PC Low && Medium"
   inputType=FunctionInput_Scalar
   sortPriority=2
  }
  n3 = FunctionInput() {
   bUsePreviewValueAsDefault=true
   desc="XB1, PS4\r\nno GBufferD / SSS Color / Fuzz Color"
   description="XB1, PS4"
   inputName="High - XB1, PS4"
   inputType=FunctionInput_Scalar
   sortPriority=1
  }
  n4 = FunctionInput() {
   bUsePreviewValueAsDefault=true
   desc="Switch\r\nhas depth buffer\r\nno roughness outlines"
   description=Switch
   inputName=Mobile Med - Switch
   inputType=FunctionInput_Scalar
   sortPriority=4
  }
  n5 = FunctionInput() {
   bUsePreviewValueAsDefault=true
   desc="PC_PerfMode, Android\r\nno depth buffer UV operation!\r\nno outlines \r\nno roughness outlines"
   description="PC_PerfMode, Android"
   inputName="Mobile Low - PC_PerfMode, Android"
   inputType=FunctionInput_Scalar
   sortPriority=5
  }
  FunctionOutput(None=ShadingPathSwitch(
    Default=QualitySwitch(Default=@n1, Low=@n2, High=@n3, Epic=@n1) {desc="fully featured GBuffers, Depth"}
    Mobile=QualitySwitch(Default=@n4, Low=@n5, Medium=@n4) {desc="limited GBuffers, limited Depth"}
   ))
 ]}
