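# Masked metaball material for the QuartzBlob backpack accessory, drawing a gloopy metallic blob.
# A Custom HLSL node raymarches four smooth-blended spheres with time-fading sine displacement, outputting opacity mask, world-space normal and step count for a debug view.
# A static switch sources spheres from dynamic parameters or procedural vector parameters; quality switches scale step count and hit threshold.
# Base color blends a Fresnel-weighted cubemap reflection with a tint.
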Material /Game/Accessories/FORT_Backpacks/Backpack_QuartzBlob/Material/M_QuartzBlob_Metaballs
graph {
 properties {
  bScreenSpaceReflections=true
  bTangentSpaceNormal=false
  bUseMaterialAttributes=true
  bUsedWithMeshDeformer=true
  bUsedWithNiagaraMeshParticles=true
  bUsedWithSkeletalMesh=true
  blendMode=BLEND_Masked
  pixelDepthOffsetMode=PDOM_Legacy
  usedShadingModels=Default Lit
 }
 nodes [
  Add(A=VectorParameter() {group=QuartzBlob, parameterName=QB Sphere1, sortPriority=40, defaultValue=<rgba -4 0 4 3>}.RGB, B=ObjectPositionWS())
  Add(A=VectorParameter() {group=QuartzBlob, parameterName=QB Sphere2, sortPriority=40, defaultValue=<rgba 4 0 4 3>}.RGB, B=ObjectPositionWS())
  Add(A=VectorParameter() {group=QuartzBlob, parameterName=QB Sphere3, sortPriority=40, defaultValue=<rgba -4 0 -4 3>}.RGB, B=ObjectPositionWS())
  Add(A=VectorParameter() {group=QuartzBlob, parameterName=QB Sphere4, sortPriority=40, defaultValue=<rgba 4 0 -4 3>}.RGB, B=ObjectPositionWS())
  n1 = Constant3Vector() {constant=<rgba 0 1 0 0>}
  n2 = Multiply(A=Time(), B=ScalarParameter() {defaultValue=0.1, parameterName=Speed})
  n3 = Constant3Vector()
  n4 = Constant4Vector() {constant=<rgba -4 0 0 5>}
  n5 = ComponentMask(None=@n4.RGBA) {b=true, g=true, r=true}
  n6 = ObjectPositionWS()
  AppendVector(A=Add(A=RotateAboutAxis(NormalizedRotationAxis=@n1, RotationAngle=@n2, PivotPoint=@n3, Position=@n5), B=@n6), B=@n4.A)
  n7 = Constant4Vector() {constant=<rgba 0 0 0 1>}
  AppendVector(A=Add(A=ComponentMask(None=@n7.RGBA) {b=true, g=true, r=true}, B=@n6), B=@n7.A)
  n8 = Constant3Vector() {constant=<rgba 1 0 0 0>}
  n9 = Constant4Vector() {constant=<rgba 0 0 4 4>}
  n10 = ComponentMask(None=@n9.RGBA) {b=true, g=true, r=true}
  AppendVector(A=Add(A=RotateAboutAxis(NormalizedRotationAxis=@n8, RotationAngle=@n2, PivotPoint=@n3, Position=@n10), B=@n6), B=@n9.A)
  n11 = Constant3Vector() {constant=<rgba 0 0 1 0>}
  n12 = Constant4Vector() {constant=<rgba 4 0 0 4>}
  n13 = ComponentMask(None=@n12.RGBA) {b=true, g=true, r=true}
  AppendVector(A=Add(A=RotateAboutAxis(NormalizedRotationAxis=@n11, RotationAngle=@n2, PivotPoint=@n3, Position=@n13), B=@n6), B=@n12.A)
  n14 = Constant()
  n15 = SetMaterialAttributes!(MaterialAttributes, Base Color=@n14, Metallic=@n14, Specular=@n14, Roughness=@n14, Opacity Mask, Normal, Emissive Color=NamedRerouteUsage()) {attributeSetTypes [
    #1
    #2
    #3
    #4
    #5
    #6
    #7
   ]}
  n16 = NamedRerouteUsage()
  n17 = SetMaterialAttributes(
   Base Color=NamedRerouteUsage()
   Metallic=NamedRerouteDeclaration(None=ScalarParameter() {defaultValue=0.9, group=QuartzBlob, parameterName=QB Metallic}) {name=Metallic, nodeColor=<rgba 0.5469 1 0 1>}
   Specular=NamedRerouteDeclaration(None=ScalarParameter() {group=QuartzBlob, parameterName=QB Specular}) {name=Specular, nodeColor=<rgba 0 1 0.1328 1>}
   Roughness=NamedRerouteDeclaration(None=ScalarParameter() {defaultValue=0.2, group=QuartzBlob, parameterName=QB Roughness}) {name=Roughness, nodeColor=<rgba 1 0 0.375 1>}
   Opacity Mask=NamedRerouteUsage()
   Normal=@n16
  ) {attributeSetTypes [
    #1
    #2
    #3
    #4
    #5
    #6
    #7
   ]}
  n18 = MaterialFunctionCall(In=StaticSwitchParameter(True=@n15, False=@n17) {group=QuartzBlob, parameterName=QB - DebugNumSteps}) {materialFunction=&MF_TechArtOverrides_accessories}
  MaterialFunctionCall(In=@n18.Result) {materialFunction=&MF_TechArtOverrides_accessories}
  n19 = TextureSampleParameterCube!(UVs=ReflectionVectorWS(), Apply View MipBias) {
   parameterName=Param
   samplerType=SAMPLERTYPE_LinearColor
   texture=&/T_quartz_Cubemap
  }
  n20 = Multiply(
   A=Fresnel(
    ExponentIn=ScalarParameter() {defaultValue=2, group=QuartzBlob, parameterName=QB Fresnel Exponent}
    BaseReflectFractionIn=ScalarParameter() {defaultValue=0.02, group=QuartzBlob, parameterName=QB Fresnel Min}
    Normal=PixelNormalWS()
   ) {baseReflectFraction=0.01, exponent=1}
   B=ScalarParameter() {defaultValue=10, group=QuartzBlob, parameterName=QB Fresnel Mult}
  )
  NamedRerouteDeclaration(None=LinearInterpolate(
    A=Multiply(A=@n19.RGB, B=@n20)
    B=VectorParameter() {group=QuartzBlob, parameterName=QB Color Tint, defaultValue=<rgba 0.18 0.4978 0.9 1>}.RGB
    Alpha=ScalarParameter() {defaultValue=0.2, group=QuartzBlob, parameterName=QB Color Tint Strength}
   ) {constAlpha=0}) {name=baseColor, nodeColor=<rgba 0.8516 0 1 1>}
  MaterialExpressionWorldPosition_0 = WorldPosition()
  MaterialExpressionCameraPositionWS_0 = CameraPositionWS()
  MaterialExpressionNormalize_0 = Normalize(VectorInput=Subtract(A=WorldPosition().XYZ, B=CameraPositionWS()))
  MaterialExpressionNamedRerouteUsage_12 = NamedRerouteUsage()
  MaterialExpressionScalarParameter_3 = ScalarParameter() {defaultValue=7, group=QuartzBlob, parameterName=QB Gloop}
  MaterialExpressionNamedRerouteUsage_8 = NamedRerouteUsage()
  MaterialExpressionNamedRerouteUsage_4 = NamedRerouteUsage()
  MaterialExpressionNamedRerouteUsage_5 = NamedRerouteUsage()
  MaterialExpressionNamedRerouteUsage_6 = NamedRerouteUsage()
  MaterialExpressionNamedRerouteUsage_15 = NamedRerouteUsage()
  MaterialExpressionNamedRerouteUsage_13 = NamedRerouteUsage()
  MaterialExpressionScalarParameter_10 = ScalarParameter() {defaultValue=40, group=QuartzBlob, parameterName=QB- maxDistance}
  n21 = Custom!(
   WorldPos=@MaterialExpressionWorldPosition_0.XYZ
   CameraPos=@MaterialExpressionCameraPositionWS_0
   CameraVector=@MaterialExpressionNormalize_0
   thresh=@MaterialExpressionNamedRerouteUsage_12
   k=@MaterialExpressionScalarParameter_3
   Sphere1=@MaterialExpressionNamedRerouteUsage_8
   Sphere2=@MaterialExpressionNamedRerouteUsage_4
   Sphere3=@MaterialExpressionNamedRerouteUsage_5
   Sphere4=@MaterialExpressionNamedRerouteUsage_6
   displacementStrength=@MaterialExpressionNamedRerouteUsage_15
   maxSteps=@MaterialExpressionNamedRerouteUsage_13
   maxDist=@MaterialExpressionScalarParameter_10
  ) {
   code="float3 curpos = WorldPos;\r\n\r\nfloat curdist  = 0;\r\nfloat epsilon = 2;\r\nfloat minstepsize = 0.001;\r\nfloat3 nn;\r\nfloat totalDist =0;\r\n\r\nint i=0;\r\nwhile(i<maxSteps)\r\n{\r\n\tnumSteps = i+1;\r\n\tcurdist =  EvaluateDistanceField(curpos,k,Sphere1,Sphere2,Sphere3, Sphere4, displacementStrength);\r\n\t\r\n\tif (curdist<thresh)\r\n\t{\r\n\t\tcurpos += CameraVector*curdist;\r\n\r\n\t\tnn.x = EvaluateDistanceField(curpos + float3(epsilon,0,0),k,Sphere1,Sphere2,Sphere3, Sphere4, displacementStrength);\r\n\t\tnn.y =  EvaluateDistanceField(curpos + float3(0,epsilon,0),k,Sphere1,Sphere2,Sphere3, Sphere4, displacementStrength);\r\n\t\tnn.z =  EvaluateDistanceField(curpos + float3(0,0,epsilon),k,Sphere1,Sphere2,Sphere3, Sphere4, displacementStrength);\r\n\t\tnn = normalize(nn);\r\n\t\treturn float4(nn.x,nn.y,nn.z,1);\r\n\t}\r\n\t//curpos += CameraVector*max(minstepsize,curdist*1);\r\n\tcurpos += CameraVector*curdist;\r\n\tminstepsize += 0.001;\t\r\n\r\n    // exit if we have marched more than maxDist\r\n    totalDist += curdist;\r\n    if (totalDist>maxDist)\r\n    {\r\n        return 0;\r\n    }\r\n\r\n\ti++;\r\n}\r\nreturn 0;"
   description=RayMarch
   outputType=CMOT_Float4
   preamble="MaterialFloat EvaluateDistanceField(\r\n    MaterialFloat3 curpos,\r\n    MaterialFloat k,\r\n    MaterialFloat4 Sphere1,\r\n    MaterialFloat4 Sphere2,\r\n    MaterialFloat4 Sphere3,\r\n    MaterialFloat4 Sphere4,\r\n    MaterialFloat displacementStrength\r\n)\r\n{\r\n    float s1 = distance(Sphere1.xyz, curpos)-Sphere1.w;\r\n    float s2 = distance(Sphere2.xyz, curpos)-Sphere2.w;\r\n    float s3 = distance(Sphere3.xyz, curpos)-Sphere3.w;\r\n    float s4 = distance(Sphere4.xyz, curpos)-Sphere4.w;\r\n\r\n    // displacement\r\n    float df = 1;\r\n    float disp = sin(curpos.x*df)*sin(curpos.y*df)*sin(curpos.z*df)*2;\r\n\r\n    float h = saturate(0.5+0.5*(s2-s1)/k);\r\n    s2 = lerp( s2, s1, h ) - k*h*(1.0-h);\r\n\r\n    h = saturate( 0.5+0.5*(s3-s2)/k);\r\n    s3 = lerp ( s3, s2, h ) -k*h*(1.0-h);\r\n\r\n    h = saturate( 0.5+0.5*(s4-s3)/k);\r\n    float curdf = lerp ( s4, s3, h ) -k*h*(1.0-h);\r\n\r\n    curdf += disp * displacementStrength;\r\n\r\n    return curdf;\r\n}"
   additionalOutputs [{outputName=numSteps, outputType=CMOT_Float1}]
   inputs [
    {
     inputName=WorldPos
     input {expression=&self:MaterialExpressionWorldPosition_0}
    }
    {
     inputName=CameraPos
     input {expression=&self:MaterialExpressionCameraPositionWS_0}
    }
    {
     inputName=CameraVector
     input {expression=&self:MaterialExpressionNormalize_0}
    }
    {
     inputName=thresh
     input {expression=&self:MaterialExpressionNamedRerouteUsage_12}
    }
    {
     inputName=k
     input {expression=&self:MaterialExpressionScalarParameter_3}
    }
    {
     inputName=Sphere1
     input {expression=&self:MaterialExpressionNamedRerouteUsage_8}
    }
    {
     inputName=Sphere2
     input {expression=&self:MaterialExpressionNamedRerouteUsage_4}
    }
    {
     inputName=Sphere3
     input {expression=&self:MaterialExpressionNamedRerouteUsage_5}
    }
    {
     inputName=Sphere4
     input {expression=&self:MaterialExpressionNamedRerouteUsage_6}
    }
    {
     inputName=displacementStrength
     input {expression=&self:MaterialExpressionNamedRerouteUsage_15}
    }
    {
     inputName=maxSteps
     input {expression=&self:MaterialExpressionNamedRerouteUsage_13}
    }
    {
     inputName=maxDist
     input {expression=&self:MaterialExpressionScalarParameter_10}
    }
   ]
  }
  NamedRerouteDeclaration(None=ComponentMask(None=@n21.return) {a=true}) {name=OpacityMask, nodeColor=<rgba 1 0.9141 0 1>}
  n22 = ComponentMask(None=VectorParameter() {parameterName=SphereOffset, defaultValue=<rgba 6 0 -10 3>}.RGB) {b=true, g=true, r=true}
  n23 = VectorParameter() {parameterName=Sphere1, defaultValue=<rgba 0 0 0 15>}
  n24 = Add(A=@n22, B=ComponentMask(None=@n23.RGB) {b=true, g=true, r=true})
  n25 = MaterialFunctionCall() {materialFunction=&ObjectScale}
  NamedRerouteDeclaration(None=AppendVector(A=TransformPosition(None=@n24) {transformType=TRANSFORMPOSSOURCE_World}, B=Multiply(A=@n23.A, B=@n25.Scale X))) {name=ProceduralSphere1, nodeColor=<rgba 0.875 1 0 1>}
  n26 = Divide(A=@n21.numSteps, B=NamedRerouteUsage())
  NamedRerouteDeclaration(None=MaterialFunctionCall(Hue Shift Percentage=Multiply(A=@n26) {constB=-0.333}, Texture=Constant3Vector() {constant=<rgba 0 1 0 0>}) {materialFunction=&HueShift}.Result) {name=StepsDebug, nodeColor=<rgba 1 0 0.1172 1>}
  n27 = VectorParameter() {parameterName=Sphere2, defaultValue=<rgba -26 -5 21 13>}
  n28 = Add(A=@n22, B=ComponentMask(None=@n27.RGB) {b=true, g=true, r=true})
  NamedRerouteDeclaration(None=AppendVector(A=TransformPosition(None=@n28) {transformType=TRANSFORMPOSSOURCE_World}, B=Multiply(A=@n27.A, B=@n25.Scale X))) {name=ProceduralSphere2, nodeColor=<rgba 0 1 0.1562 1>}
  n29 = VectorParameter() {parameterName=Sphere3, defaultValue=<rgba 7 0 30 10>}
  n30 = Add(A=@n22, B=ComponentMask(None=@n29.RGB) {b=true, g=true, r=true})
  NamedRerouteDeclaration(None=AppendVector(A=TransformPosition(None=@n30) {transformType=TRANSFORMPOSSOURCE_World}, B=Multiply(A=@n29.A, B=@n25.Scale X))) {name=ProceduralSphere3, nodeColor=<rgba 1 0.1641 0 1>}
  n31 = VectorParameter() {parameterName=Sphere4, defaultValue=<rgba -12 25 -10 6>}
  n32 = Add(A=@n22, B=ComponentMask(None=@n31.RGB) {b=true, g=true, r=true})
  NamedRerouteDeclaration(None=AppendVector(A=TransformPosition(None=@n32) {transformType=TRANSFORMPOSSOURCE_World}, B=Multiply(A=@n31.A, B=@n25.Scale X))) {name=ProceduralSphere4, nodeColor=<rgba 0.4766 1 0 1>}
  n33 = ScalarParameter() {
   defaultValue=0.1
   desc="if the marched point is below this distance from the df it is counted as being within the df. "
   group=QuartzBlob
   parameterName=QB Threshold
  }
  n34 = ScalarParameter() {defaultValue=0.1, group=QuartzBlob Low, parameterName=QB Threshold Low}
  NamedRerouteDeclaration(None=MaterialFunctionCall!(
    "Epic - PC-High && Epic, XSX, PS5"=@n33
    "High - XB1, PS4"=@n33
    "Low - PC Low && Medium"=ScalarParameter() {defaultValue=0.1, group=QuartzBlob Mid, parameterName=QB Threshold Mid}
    --------
    Mobile Med - Switch=@n34
    "Mobile Low - PC_PerfMode, Android"=@n34
   ) {materialFunction=&/MF_QuartzBlob_ScalarQualitySwitch}.Result) {name=Threshold, nodeColor=<rgba 1 0.1172 0 1>}
  n35 = ScalarParameter() {defaultValue=32, group=QuartzBlob, parameterName=QB maxSteps}
  n36 = ScalarParameter() {defaultValue=16, group=QuartzBlob Low, parameterName=QB maxSteps Low}
  NamedRerouteDeclaration(None=MaterialFunctionCall(
    "Epic - PC-High && Epic, XSX, PS5"=@n35
    "High - XB1, PS4"=@n35
    "Low - PC Low && Medium"=ScalarParameter() {defaultValue=24, group=QuartzBlob Mid, parameterName=QB maxSteps Mid}
    Mobile Med - Switch=@n36
    "Mobile Low - PC_PerfMode, Android"=@n36
   ) {materialFunction=&MF_QuartzBlob_ScalarQualitySwitch}.Result) {name=MaxSteps, nodeColor=<rgba 1 0.3281 0 1>}
  n37 = Saturate(None=Subtract(A=Time(), B=ParticleColor().R))
  NamedRerouteDeclaration(None=Multiply(A=OneMinus(None=@n37), B=ScalarParameter() {defaultValue=0.2, group=QuartzBlob, parameterName=QB displacementStrength}) {constB=0.3}) {name=Displacement, nodeColor=<rgba 0 1 0.5781 1>}
  NamedRerouteDeclaration(None=ComponentMask(None=@n21.return) {b=true, g=true, r=true}) {name=Normal, nodeColor=<rgba 0.5937 1 0 1>}
  n38 = DynamicParameter()
  n39 = TransformPosition(None=MaterialFunctionCall(X=@n38.R, Y=@n38.R, Z=@n38.R) {materialFunction=&MakeFloat3}.Result) {transformType=TRANSFORMPOSSOURCE_World}
  NamedRerouteDeclaration(None=StaticSwitchParameter(True=NamedRerouteUsage(), False=AppendVector(A=@n39, B=@n38.A)) {group=QuartzBlob, parameterName=QB proceduralSpheres, sortPriority=38}) {name=Sphere1, nodeColor=<rgba 0.3828 0 1 1>}
  n40 = DynamicParameter() {parameterIndex=1}
  n41 = TransformPosition(None=MaterialFunctionCall(X=@n40.R, Y=@n40.R, Z=@n40.R) {materialFunction=&MakeFloat3}.Result) {transformType=TRANSFORMPOSSOURCE_World}
  NamedRerouteDeclaration(None=StaticSwitchParameter(True=NamedRerouteUsage(), False=AppendVector(A=@n41, B=@n40.A)) {group=QuartzBlob, parameterName=QB proceduralSpheres, sortPriority=38}) {name=Sphere2, nodeColor=<rgba 1 0.2578 0 1>}
  n42 = DynamicParameter() {parameterIndex=2}
  n43 = TransformPosition(None=MaterialFunctionCall(X=@n42.R, Y=@n42.R, Z=@n42.R) {materialFunction=&MakeFloat3}.Result) {transformType=TRANSFORMPOSSOURCE_World}
  NamedRerouteDeclaration(None=StaticSwitchParameter(True=NamedRerouteUsage(), False=AppendVector(A=@n43, B=@n42.A)) {group=QuartzBlob, parameterName=QB proceduralSpheres, sortPriority=38}) {name=Sphere3, nodeColor=<rgba 0 0.4141 1 1>}
  n44 = DynamicParameter() {parameterIndex=3}
  n45 = TransformPosition(None=MaterialFunctionCall(X=@n44.R, Y=@n44.R, Z=@n44.R) {materialFunction=&MakeFloat3}.Result) {transformType=TRANSFORMPOSSOURCE_World}
  NamedRerouteDeclaration(None=StaticSwitchParameter(True=NamedRerouteUsage(), False=AppendVector(A=@n45, B=@n44.A)) {group=QuartzBlob, parameterName=QB proceduralSpheres, sortPriority=38}) {name=Sphere4, nodeColor=<rgba 0.2891 1 0 1>}
  n46 = MaterialFunctionCall() {materialFunction=&CameraDirectionVector}
  n47 = Multiply(A=ScalarParameter() {defaultValue=0.3, group=QuartzBlob, parameterName=QB EdgeCleanup, sliderMax=1}) {constB=-1}
  Normalize(VectorInput=Add(A=@n16, B=Multiply(A=@n46.Result, B=@n47) {constB=-0.2}))
  Saturate(None=@n20)
  ScalarParameter() {group=1 Meteorwomen Reactivity, parameterName=GameTimeFromBP}
 ]
}
